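# Prototype DentureSay item animation Blueprint for the Fortnite player skeleton, overriding jump, fall and final-pose layers.
# Jump layer state machine plays forward/backward boosted jumps; the full-body version's entry transitions are disabled, and fall level 2 plays a jump-loop sequence.
# Update enables a pitch-driven, pelvis-filtered aim offset during the air-attack montage.
# It also ramps a fall-pose alpha, but that blend and the pre-IK bone rotations are disconnected.
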AnimBlueprint /DentureSay/Assets/Proto/DentureSay_AnimBP
target_skeleton=&Fortnite_M_Avg_Player_Skeleton
graphs [
 AnimGraph {nodes [Output Pose()]}
 EventGraph {nodes [
   n1 = Event BlueprintUpdateAnimation()
   > Set Delta Time X(Delta Time X=@n1.DeltaTimeX)
   > Is Valid(InputObject=Get FortPlayerPawn() -> (FortPlayerPawn:Fort Player Pawn Object Reference))
   > n2 = Sequence!(execute) -> (then_0, then_1, then_2:Exec)
   n3 = Get FortPlayerPawn()
   > n4 = Set PlayingInAirAttachMontage(PlayingInAirAttachMontage:bool="Equal (Object)"(
     A=GetCurrentMontage!(self:Character Object Reference=@n3) -> (ReturnValue:Anim Montage Object Reference)
     B=&/DentureSay_AirAttack_Montage
    )) -> (Output_Get:bool)
   > n5 = Branch(Condition=@n4)
   > Set AirAttack_AO_Alpha(AirAttack_AO_Alpha:float=1) -> (Output_Get:float)
   Set AirAttack_AO_Alpha(execute=@n5.else, AirAttack_AO_Alpha=0)
   n6 = Get PlayingInAirAttachMontage() -> (PlayingInAirAttachMontage:bool)
   n7 = Branch(execute=@n2.then_1, Condition=AND Boolean(A=Get bIsInAir() -> (bIsInAir:bool), B=NOT Boolean(A=@n6)))
   > n8 = Branch(Condition="In Range (Float)"!(
     Value:double=Property Access!() -> (Value:float=~) {is=title K2Node_PropertyAccess}
     Min:double=-1100
     Max:double=200
     InclusiveMin:bool=true
     InclusiveMax:bool=true
    ) -> (ReturnValue:bool))
   n9 = float * float(A=Get Delta Time X(), B=1.75) {is=title K2Node_PromotableOperator}
   n10 = Get JumpFallPoseAdditive() -> (JumpFallPoseAdditive:double)
   > Set JumpFallPoseAdditive(JumpFallPoseAdditive:double="Clamp (Float)"(Value=float + float(A=@n10, B=@n9), Min=0, Max=1)) -> (Output_Get:double)
   n11 = float * float(A=Get Delta Time X(), B=2)
   n12 = Get JumpFallPoseAdditive()
   Set JumpFallPoseAdditive(execute=@n8.else, JumpFallPoseAdditive="Clamp (Float)"(Value=float - float(A=@n12, B=@n11), Min=0, Max=1))
   Set JumpFallPoseAdditive(execute=@n7.else, JumpFallPoseAdditive=0)
   Branch(Condition=Get bIsJumping())
   n13 = float * float(A=Get Delta Time X(), B=1)
   n14 = Get JumpFallPoseAdditive()
   > Set JumpFallPoseAdditive(JumpFallPoseAdditive="Clamp (Float)"(Value=float + float(A=@n14, B=@n13), Min=0, Max=1))
   PrintString!(execute, InString="To String (Float)"(InDouble=Get JumpFallPoseAdditive()), bPrintToScreen, bPrintToLog, TextColor, Duration, Key) -> (then)
  ]}
 ItemAimOffset {nodes [
   Input Pose!() -> (Pose, AimOffsetAlpha:double, Yaw:double, Pitch:double) {name=InPose_AimOffset}
   Input Pose!() -> (Pose) {name=InPose_UpperLowerPreMeleeAO}
   Output Pose() {layerGroup=Item Override}
  ]}
 ItemFallAdditive {nodes [
   Input Pose() {name=InPoseItemFallAdditive}
   Output Pose() {layerGroup=Item Override}
  ]}
 ItemFallLandAdditive {nodes [
   Input Pose() {name=InPoseItemFallLandAdditive}
   Output Pose() {layerGroup=Item Override}
  ]}
 ItemFallLevel2Override {nodes [
   Input Pose() {name=InPose_FallLevel2}
   Output Pose(Result=Sequence Player 'DentureSay_JumpLoop'() {bLoopAnimation=false, sequence=&/DentureSay_JumpLoop}) {layerGroup=Item Override}
  ]}
 ItemFinalPoseOverride {nodes [
   n1 = Layered blend per bone(
    BasePose=Use cached pose 'FinalPoseCache'()
    BlendPoses_0=AimOffset Player 'DentureSay_AirAttack_AO'!(
     Y:float=Get AimPitch()
     BasePose:Pose Link Structure=Use cached pose 'FinalPoseCache'()
     Alpha:float=Get AirAttack_AO_Alpha() -> (AirAttack_AO_Alpha:float)
    ) -> (Pose:Pose Link Structure) {blendSpace=&/DentureSay_AirAttack_AO}
   ) {layerSetup [{branchFilters [{boneName=Pelvis}, {blendDepth=-1, boneName=Spine_01}, {blendDepth=-1, boneName=thigh_l}, {blendDepth=-1, boneName=thigh_r}]}]}
   Output Pose(Result=@n1) {layerGroup=Item Override}
   Save cached pose 'FinalPoseCache'(Pose=Input Pose() {name=InFinalPose}) {cacheName=FinalPoseCache}
  ]}
 ItemFullBodyOverride {nodes [
   AnimGraphNode_StateMachine_0 = jump loop() {is=title AnimGraphNode_StateMachine, bAllowConduitEntryStates=true}
   Output Pose(Result=@AnimGraphNode_StateMachine_0) {layerGroup=Item Override}
   Save cached pose 'FullBodyCache'(Pose=Input Pose() {name=InPoseFullBody}) {cacheName=FullBodyCache}
   Two Way Blend(
    B=Sequence Player 'DentureSay_Fall_Pose'() {bLoopAnimation=false, sequence=&/DentureSay_Fall_Pose}
    Alpha=Get JumpFallPoseAdditive()
   )
  ]}
 jump loop {
  entry=Passthrough
  path=ItemFullBodyOverride.AnimGraphNode_StateMachine_0.jump loop
  ids=id
  nodes [
   AnimStateNode_0 = Passthrough() {is=title AnimStateNode}
   AnimStateNode_2 = BoostedJumpW() {is=title AnimStateNode}
   AnimStateNode_3 = BoostedJumpN() {is=title AnimStateNode}
   AnimStateNode_4 = BoostedJumpS() {is=title AnimStateNode}
   AnimStateNode_5 = BoostedJumpE() {is=title AnimStateNode}
   AnimStateNode_6 = Jumping() {is=title AnimStateNode}
   AnimStateTransitionNode_1 = AnimStateTransitionNode() {bAllowInertializationForSelfTransitions=true, bDisabled=true}
   AnimStateTransitionNode_10 = AnimStateTransitionNode() {bAllowInertializationForSelfTransitions=true, bDisabled=true, blendMode=ExpOut, crossfadeDuration=0.05, logicType=TLT_Inertialization}
   AnimStateTransitionNode_11 = AnimStateTransitionNode() {bAllowInertializationForSelfTransitions=true, bDisabled=true, crossfadeDuration=0.1}
   AnimStateTransitionNode_12 = AnimStateTransitionNode() {
    bAllowInertializationForSelfTransitions=true
    bAutomaticRuleBasedOnSequencePlayerInState=true
    crossfadeDuration=0.3
    logicType=TLT_Inertialization
   }
   AnimStateTransitionNode_14 = AnimStateTransitionNode() {bAllowInertializationForSelfTransitions=true, bDisabled=true}
   AnimStateTransitionNode_15 = AnimStateTransitionNode() {bAllowInertializationForSelfTransitions=true, bAutomaticRuleBasedOnSequencePlayerInState=true}
   AnimStateTransitionNode_16 = AnimStateTransitionNode() {bAllowInertializationForSelfTransitions=true}
   AnimStateTransitionNode_2 = AnimStateTransitionNode() {bAllowInertializationForSelfTransitions=true, bDisabled=true}
   AnimStateTransitionNode_3 = AnimStateTransitionNode() {bAllowInertializationForSelfTransitions=true, crossfadeDuration=0.3}
   AnimStateTransitionNode_4 = AnimStateTransitionNode() {bAllowInertializationForSelfTransitions=true, crossfadeDuration=0.3}
   AnimStateTransitionNode_5 = AnimStateTransitionNode() {bAllowInertializationForSelfTransitions=true, bAutomaticRuleBasedOnSequencePlayerInState=true, crossfadeDuration=0.3}
   AnimStateTransitionNode_6 = AnimStateTransitionNode() {bAllowInertializationForSelfTransitions=true, crossfadeDuration=0.3, logicType=TLT_Inertialization}
   AnimStateTransitionNode_7 = AnimStateTransitionNode() {bAllowInertializationForSelfTransitions=true, crossfadeDuration=0.3}
   AnimStateTransitionNode_8 = AnimStateTransitionNode() {bAllowInertializationForSelfTransitions=true, bAutomaticRuleBasedOnSequencePlayerInState=true, crossfadeDuration=0.3}
   AnimStateTransitionNode_9 = AnimStateTransitionNode() {bAllowInertializationForSelfTransitions=true, bAutomaticRuleBasedOnSequencePlayerInState=true, crossfadeDuration=0.3}
  ]
  links [
   {
    dst=Passthrough
    rule=ItemFullBodyOverride.AnimGraphNode_StateMachine_0.jump loop.AnimStateTransitionNode_6.Transition
    src=BoostedJumpE
    via=AnimStateTransitionNode_3
   }
   {
    dst=Passthrough
    rule=ItemFullBodyOverride.AnimGraphNode_StateMachine_0.jump loop.AnimStateTransitionNode_8.Transition
    src=BoostedJumpE
    via=AnimStateTransitionNode_8
   }
   {
    dst=Passthrough
    rule=ItemFullBodyOverride.AnimGraphNode_StateMachine_0.jump loop.AnimStateTransitionNode_12.Transition
    src=BoostedJumpN
    via=AnimStateTransitionNode_12
   }
   {
    dst=Passthrough
    rule=ItemFullBodyOverride.AnimGraphNode_StateMachine_0.jump loop.AnimStateTransitionNode_6.Transition
    src=BoostedJumpN
    via=AnimStateTransitionNode_6
   }
   {
    dst=Passthrough
    rule=ItemFullBodyOverride.AnimGraphNode_StateMachine_0.jump loop.AnimStateTransitionNode_5.Transition
    src=BoostedJumpS
    via=AnimStateTransitionNode_5
   }
   {
    dst=Passthrough
    rule=ItemFullBodyOverride.AnimGraphNode_StateMachine_0.jump loop.AnimStateTransitionNode_6.Transition
    src=BoostedJumpS
    via=AnimStateTransitionNode_7
   }
   {
    dst=Passthrough
    rule=ItemFullBodyOverride.AnimGraphNode_StateMachine_0.jump loop.AnimStateTransitionNode_6.Transition
    src=BoostedJumpW
    via=AnimStateTransitionNode_4
   }
   {
    dst=Passthrough
    rule=ItemFullBodyOverride.AnimGraphNode_StateMachine_0.jump loop.AnimStateTransitionNode_9.Transition
    src=BoostedJumpW
    via=AnimStateTransitionNode_9
   }
   {
    dst=Passthrough
    rule=ItemFullBodyOverride.AnimGraphNode_StateMachine_0.jump loop.AnimStateTransitionNode_15.Transition
    src=Jumping
    via=AnimStateTransitionNode_15
   }
   {
    dst=Passthrough
    rule=ItemFullBodyOverride.AnimGraphNode_StateMachine_0.jump loop.AnimStateTransitionNode_6.Transition
    src=Jumping
    via=AnimStateTransitionNode_16
   }
   {
    dst=BoostedJumpE
    rule=ItemFullBodyOverride.AnimGraphNode_StateMachine_0.jump loop.AnimStateTransitionNode_2.Transition
    src=Passthrough
    via=AnimStateTransitionNode_2
   }
   {
    dst=BoostedJumpN
    rule=ItemFullBodyOverride.AnimGraphNode_StateMachine_0.jump loop.AnimStateTransitionNode_10.Transition
    src=Passthrough
    via=AnimStateTransitionNode_10
   }
   {
    dst=BoostedJumpS
    rule=ItemFullBodyOverride.AnimGraphNode_StateMachine_0.jump loop.AnimStateTransitionNode_11.Transition
    src=Passthrough
    via=AnimStateTransitionNode_11
   }
   {
    dst=BoostedJumpW
    rule=ItemFullBodyOverride.AnimGraphNode_StateMachine_0.jump loop.AnimStateTransitionNode_1.Transition
    src=Passthrough
    via=AnimStateTransitionNode_1
   }
   {
    dst=Jumping
    rule=ItemFullBodyOverride.AnimGraphNode_StateMachine_0.jump loop.AnimStateTransitionNode_14.Transition
    src=Passthrough
    via=AnimStateTransitionNode_14
   }
  ]
 }
 Passthrough {
  path=ItemFullBodyOverride.AnimGraphNode_StateMachine_0.jump loop.AnimStateNode_0.Passthrough
  nodes [Output Animation Pose(Result=Use cached pose 'FullBodyCache'())]
 }
 BoostedJumpW {
  path=ItemFullBodyOverride.AnimGraphNode_StateMachine_0.jump loop.AnimStateNode_2.BoostedJumpW
  nodes [Output Animation Pose(Result=Sequence Player 'Boost_Jump_Left'() {bLoopAnimation=false, sequence=&/Boost_Jump_Left})]
 }
 BoostedJumpN {
  path=ItemFullBodyOverride.AnimGraphNode_StateMachine_0.jump loop.AnimStateNode_3.BoostedJumpN
  nodes [Output Animation Pose(Result=Sequence Player 'Boost_Jump_Forward'() {bLoopAnimation=false, sequence=&/Boost_Jump_Forward})]
 }
 BoostedJumpS {
  path=ItemFullBodyOverride.AnimGraphNode_StateMachine_0.jump loop.AnimStateNode_4.BoostedJumpS
  nodes [Output Animation Pose(Result=Sequence Player 'Boost_Jump_Backward'() {bLoopAnimation=false, sequence=&/Boost_Jump_Backward})]
 }
 BoostedJumpE {
  path=ItemFullBodyOverride.AnimGraphNode_StateMachine_0.jump loop.AnimStateNode_5.BoostedJumpE
  nodes [Output Animation Pose(Result=Sequence Player 'Boost_Jump_Right'() {bLoopAnimation=false, sequence=&/Boost_Jump_Right})]
 }
 Jumping {
  path=ItemFullBodyOverride.AnimGraphNode_StateMachine_0.jump loop.AnimStateNode_6.Jumping
  nodes [Output Animation Pose(Result=Blendspace Player 'Boost_Jump_BS'(X=Property Access() -> (Value=~), Y=0) {bLoop=false, blendSpace=&/Boost_Jump_BS})]
 }
 Transition {
  path=ItemFullBodyOverride.AnimGraphNode_StateMachine_0.jump loop.AnimStateTransitionNode_1.Transition
  nodes [Result(bCanEnterTransition=Get BoostedJumpW() -> (BoostedJumpW:bool))]
 }
 Transition {
  path=ItemFullBodyOverride.AnimGraphNode_StateMachine_0.jump loop.AnimStateTransitionNode_10.Transition
  nodes [Result(bCanEnterTransition=Get BoostedJumpN() -> (BoostedJumpN:bool))]
 }
 Transition {
  path=ItemFullBodyOverride.AnimGraphNode_StateMachine_0.jump loop.AnimStateTransitionNode_11.Transition
  nodes [Result(bCanEnterTransition=Get BoostedJumpS() -> (BoostedJumpS:bool))]
 }
 Transition {path=ItemFullBodyOverride.AnimGraphNode_StateMachine_0.jump loop.AnimStateTransitionNode_12.Transition, nodes [Result()]}
 Transition {
  path=ItemFullBodyOverride.AnimGraphNode_StateMachine_0.jump loop.AnimStateTransitionNode_14.Transition
  nodes [Result(bCanEnterTransition=Get bIsJumping())]
 }
 Transition {path=ItemFullBodyOverride.AnimGraphNode_StateMachine_0.jump loop.AnimStateTransitionNode_15.Transition, nodes [Result()]}
 Transition {
  path=ItemFullBodyOverride.AnimGraphNode_StateMachine_0.jump loop.AnimStateTransitionNode_2.Transition
  nodes [Result(bCanEnterTransition=Get BoostedJumpE() -> (BoostedJumpE:bool))]
 }
 Transition {path=ItemFullBodyOverride.AnimGraphNode_StateMachine_0.jump loop.AnimStateTransitionNode_5.Transition, nodes [Result()]}
 Transition {
  path=ItemFullBodyOverride.AnimGraphNode_StateMachine_0.jump loop.AnimStateTransitionNode_6.Transition
  nodes [
   n1 = Get bIsInAir()
   Result(bCanEnterTransition=OR Boolean(A=NOT Boolean(A=@n1), B=Get bIsFalling()))
  ]
 }
 Transition {path=ItemFullBodyOverride.AnimGraphNode_StateMachine_0.jump loop.AnimStateTransitionNode_8.Transition, nodes [Result()]}
 Transition {path=ItemFullBodyOverride.AnimGraphNode_StateMachine_0.jump loop.AnimStateTransitionNode_9.Transition, nodes [Result()]}
 ItemIdleOverride {nodes [
   Input Pose() {name=InPose_Idle}
   Output Pose() {layerGroup=Item Override}
  ]}
 ItemJetPackJumpAdditive {nodes [Output Pose() {layerGroup=Item Override}]}
 ItemJetPackStartAdditive {nodes [Output Pose() {layerGroup=Item Override}]}
 ItemJumpLoopAdditive {nodes [
   Input Pose() {name=InPoseItemJumpLoopAdditive}
   Output Pose() {layerGroup=Item Override}
  ]}
 ItemJumpOverride {nodes [
   AnimGraphNode_StateMachine_0 = jump loop() {bAllowConduitEntryStates=true}
   Output Pose(Result=@AnimGraphNode_StateMachine_0) {layerGroup=Item Override}
   Save cached pose 'jumpentry'(Pose=Input Pose() {name=InPoseJump}) {cacheName=jumpentry}
  ]}
 jump loop {
  entry=Passthrough
  path=ItemJumpOverride.AnimGraphNode_StateMachine_0.jump loop
  ids=id
  nodes [
   AnimStateNode_0 = Passthrough()
   AnimStateNode_2 = BoostedJumpS()
   AnimStateNode_3 = BoostedJumpN()
   AnimStateTransitionNode_10 = AnimStateTransitionNode() {bAllowInertializationForSelfTransitions=true, crossfadeDuration=0.1}
   AnimStateTransitionNode_11 = AnimStateTransitionNode() {bAllowInertializationForSelfTransitions=true, crossfadeDuration=0.1}
   AnimStateTransitionNode_12 = AnimStateTransitionNode() {
    bAllowInertializationForSelfTransitions=true
    bAutomaticRuleBasedOnSequencePlayerInState=true
    crossfadeDuration=0.5
    logicType=TLT_Inertialization
   }
   AnimStateTransitionNode_5 = AnimStateTransitionNode() {bAllowInertializationForSelfTransitions=true, bAutomaticRuleBasedOnSequencePlayerInState=true}
   AnimStateTransitionNode_6 = AnimStateTransitionNode() {bAllowInertializationForSelfTransitions=true, crossfadeDuration=0.3}
   AnimStateTransitionNode_7 = AnimStateTransitionNode() {bAllowInertializationForSelfTransitions=true, crossfadeDuration=0.3}
  ]
  links [
   {
    dst=Passthrough
    rule=ItemJumpOverride.AnimGraphNode_StateMachine_0.jump loop.AnimStateTransitionNode_12.Transition
    src=BoostedJumpN
    via=AnimStateTransitionNode_12
   }
   {
    dst=Passthrough
    rule=ItemJumpOverride.AnimGraphNode_StateMachine_0.jump loop.AnimStateTransitionNode_6.Transition
    src=BoostedJumpN
    via=AnimStateTransitionNode_6
   }
   {
    dst=Passthrough
    rule=ItemJumpOverride.AnimGraphNode_StateMachine_0.jump loop.AnimStateTransitionNode_5.Transition
    src=BoostedJumpS
    via=AnimStateTransitionNode_5
   }
   {
    dst=Passthrough
    rule=ItemJumpOverride.AnimGraphNode_StateMachine_0.jump loop.AnimStateTransitionNode_6.Transition
    src=BoostedJumpS
    via=AnimStateTransitionNode_7
   }
   {
    dst=BoostedJumpN
    rule=ItemJumpOverride.AnimGraphNode_StateMachine_0.jump loop.AnimStateTransitionNode_10.Transition
    src=Passthrough
    via=AnimStateTransitionNode_10
   }
   {
    dst=BoostedJumpS
    rule=ItemJumpOverride.AnimGraphNode_StateMachine_0.jump loop.AnimStateTransitionNode_11.Transition
    src=Passthrough
    via=AnimStateTransitionNode_11
   }
  ]
 }
 Passthrough {
  path=ItemJumpOverride.AnimGraphNode_StateMachine_0.jump loop.AnimStateNode_0.Passthrough
  nodes [Output Animation Pose(Result=Use cached pose 'jumpentry'())]
 }
 BoostedJumpS {
  path=ItemJumpOverride.AnimGraphNode_StateMachine_0.jump loop.AnimStateNode_2.BoostedJumpS
  nodes [Output Animation Pose(Result=Sequence Player 'Boost_Jump_Backward'() {bLoopAnimation=false, sequence=&Boost_Jump_Backward})]
 }
 BoostedJumpN {
  path=ItemJumpOverride.AnimGraphNode_StateMachine_0.jump loop.AnimStateNode_3.BoostedJumpN
  nodes [Output Animation Pose(Result=Sequence Player 'Boost_Jump_Forward'() {bLoopAnimation=false, sequence=&Boost_Jump_Forward})]
 }
 Transition {
  path=ItemJumpOverride.AnimGraphNode_StateMachine_0.jump loop.AnimStateTransitionNode_10.Transition
  nodes [Result(bCanEnterTransition=Get BoostedJumpN())]
 }
 Transition {
  path=ItemJumpOverride.AnimGraphNode_StateMachine_0.jump loop.AnimStateTransitionNode_11.Transition
  nodes [Result(bCanEnterTransition=Get BoostedJumpS())]
 }
 Transition {path=ItemJumpOverride.AnimGraphNode_StateMachine_0.jump loop.AnimStateTransitionNode_12.Transition, nodes [Result()]}
 Transition {path=ItemJumpOverride.AnimGraphNode_StateMachine_0.jump loop.AnimStateTransitionNode_5.Transition, nodes [Result()]}
 Transition {
  path=ItemJumpOverride.AnimGraphNode_StateMachine_0.jump loop.AnimStateTransitionNode_6.Transition
  nodes [
   n1 = Get bIsInAir()
   Result(bCanEnterTransition=OR Boolean(A=NOT Boolean(A=@n1), B=Get bIsFalling()))
  ]
 }
 ItemJumpUpAdditive {nodes [
   Input Pose() {name=InPoseItemJumpUpAdditive}
   Output Pose() {layerGroup=Item Override}
  ]}
 ItemPreAimOffsets {nodes [
   Input Pose() {name=InPosePreAimOffsets}
   Output Pose() {layerGroup=Item Override}
  ]}
 ItemPreIK {nodes [
   n1 = "Transform (Modify) Bone - Bone: pelvis"!(ComponentPose, Alpha=1, Translation="0,0,0", Rotation_Roll:float=-50, Rotation_Pitch:float=0, Rotation_Yaw:float=0, Scale="1,1,1") -> (Pose) {rotationMode=BMM_Additive, rotationSpace=BCS_BoneSpace, alphaBoolBlend {blendOption=Linear}, boneToModify {boneName=pelvis}}
   "Transform (Modify) Bone - Bone: Spine_01"!(ComponentPose=@n1, Alpha=1, Translation="0,0,0", Rotation_Roll:float=50, Rotation_Pitch:float=0, Rotation_Yaw:float=0, Scale="1,1,1") -> (Pose) {rotationMode=BMM_Additive, rotationSpace=BCS_BoneSpace, alphaBoolBlend {blendOption=Linear}, boneToModify {boneName=Spine_01}}
   Output Pose() {layerGroup=Item Override}
   Save cached pose 'PreIkCache'(Pose=Input Pose() {name=InPose_PreIK}) {cacheName=PreIkCache}
   Use cached pose 'PreIkCache'()
  ]}
 ItemRightHandIK {nodes [
   Input Pose() {name=InPosePreRightHandIK}
   Input Pose() {name=InPosePostRightHandIK}
   Output Pose() {layerGroup=Item Override}
  ]}
 ItemSkeletalControl {nodes [
   Input Pose()
   Output Pose() {layerGroup=Item Override}
  ]}
 ItemSwimJumpFallLoop {nodes [Output Pose() {layerGroup=Item Override}]}
 ItemSwimJumpSurfaceEnd {nodes [Output Pose() {layerGroup=Item Override}]}
 ItemSwimJumpSurfaceLoop {nodes [Output Pose() {layerGroup=Item Override}]}
 ItemUpperBody {nodes [
   Input Pose!() -> (Pose, InputParam:Fort Anim Input Adjusted Aim Structure) {name=InPoseUpperBody}
   Output Pose() {layerGroup=Item Override}
  ]}
 ItemZipLineStartAdditive {nodes [Output Pose() {layerGroup=Item Override}]}
]
variables [
 AirAttack_AO_Alpha {type=number, default=1}
 BoostedJumpE {type=boolean}
 BoostedJumpN {type=boolean}
 BoostedJumpS {type=boolean}
 BoostedJumpW {type=boolean}
 Delta Time X {type=number}
 JumpFallPoseAdditive {type=number}
 PlayingInAirAttachMontage {type=boolean}
]
